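# M_DentureSay_HealingOrb_Quadrant_1 is a minimal Default Lit material, named as one quadrant of a prototype healing orb.
# Its graph holds a single vector parameter with a cyan default color and zero alpha.
# No connection from that parameter to a material output is present.
# Automatic setting of usage flags in the editor is turned off.
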Material /DentureSay/Assets/Proto/HealingOrb/M_DentureSay_HealingOrb_Quadrant_1
graph {
 properties {bAutomaticallySetUsageInEditor=false, usedShadingModels=Default Lit}
 nodes [VectorParameter() {defaultValue=<rgba 0.12 0.78 0.92 0>}]
}
