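# Default Lit material for quadrant 2 of the DentureSay prototype healing orb, as its name and path indicate.
# Its graph holds a single vector parameter with a blue default color and shows no connection from it to any material output.
# Automatic setting of usage flags in the editor is turned off.
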Material /DentureSay/Assets/Proto/HealingOrb/M_DentureSay_HealingOrb_Quadrant_2
graph {
 properties {bAutomaticallySetUsageInEditor=false, usedShadingModels=Default Lit}
 nodes [VectorParameter() {defaultValue=<rgba 0.13 0.32 0.85 0>}]
}
