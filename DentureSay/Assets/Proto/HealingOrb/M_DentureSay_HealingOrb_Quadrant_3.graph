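# Prototype Default Lit material named for the third quadrant of the DentureSay healing orb.
# The graph holds a single vector parameter with an orange default color and shows no connection from it to any material output.
# Automatic setting of usage flags in the editor is turned off.
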Material /DentureSay/Assets/Proto/HealingOrb/M_DentureSay_HealingOrb_Quadrant_3
graph {
 properties {bAutomaticallySetUsageInEditor=false, usedShadingModels=Default Lit}
 nodes [VectorParameter() {defaultValue=<rgba 0.96 0.55 0.12 0>}]
}
